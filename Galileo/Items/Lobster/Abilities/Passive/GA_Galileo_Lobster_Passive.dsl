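; Passive ability for the Lobster lightsaber that manages its looping audio, weapon holstering and HUD prompts until the player dies.
; Gameplay tag wait tasks holster it while carrying players or in roofed vehicle seats.
; Movement-mode changes holster it on surface swimming after a net sync, with landing or tag-query hooks for leaving water.
; Non-authority instances watch tags to refresh attack, block and dodge prompts, retrying next tick if the weapon cast fails.
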(event Ability|EventActivateAbility
  (bind _holsterid (Variables|Default|GetHolsterID))
  (bind _playerpawn (Variables|Default|GetPlayerPawn))
  (Development|PrintString "Activate")
  (bind _asfort_player_pawn (Utilities|Casting|CastToFortPlayerPawn (Ability|GetAvatarActorfromActorInfo)))
  (Variables|Default|SetPlayerPawn _asfort_player_pawn)
  ;[1 1 1] Looping Audio
  (bind _returnvalue (Ability|ApplyGameplayEffectToOwner (Variables|Default|GetGE_LoopingAudio)))
  (Variables|Default|SetLoopingAudioHandle _returnvalue)
  ;[1 1 1] End the ability when the player dies
  (EventDispatchers|AssignOnDied (Variables|Default|GetPlayerPawn) (delegate Custom|OnDied_Event_0))
  (Utilities|FlowControl|Sequence
    (:then_0
      (Character|BindEventtoMovementModeChangedDelegate (Variables|Default|GetPlayerPawn) (EventDispatchers|CreateEvent :!event Custom|Movement%20Mode%20Changed%20Delegate)))
    (:then_1
      ;[1 1 1] Holster Weapon while Carrying Players
      (Ability|Tasks|WaitGameplayTagAdd "(TagName=\"Gameplay.Action.Player.CarryingPlayer\")" _playerpawn
        (:then
          (Ability|Tasks|WaitGameplayTagRemove "(TagName=\"Gameplay.Action.Player.CarryingPlayer\")" _playerpawn
            (:then
              ;[1 1 1] Holster weapon while in roofed vehicle seat
              (Ability|Tasks|WaitGameplayTagAdd "(TagName=\"Gameplay.Status.InVehicle.Seat.Roofed\")" _playerpawn
                (:then
                  (Ability|Tasks|WaitGameplayTagRemove "(TagName=\"Gameplay.Status.InVehicle.Seat.Roofed\")" _playerpawn
                    (:Removed
                      (Weapon|SetHolsterWeapon _playerpawn 0 false true _holsterid))))
                (:Added
                  (Weapon|SetHolsterWeapon _playerpawn 0 true true _holsterid)
                  (bind _asb_athena_galileo_lobster_parent (Utilities|Casting|CastToB_Athena_Galileo_Lobster_Parent (Weapon|GetCurrentWeapon _playerpawn)))
                  (Class|BAthenaGalileoLobsterParent|SetBladeVisibility _asb_athena_galileo_lobster_parent false :!class /Galileo/Items/Lobster/Actors/B_Athena_Galileo_Lobster_Parent.B_Athena_Galileo_Lobster_Parent_C))))
            (:Removed
              (if (not (GameplayTags|HasAnyMatchingGameplayTags _playerpawn (Variables|Default|GetSwimmingTags)))
                (Development|PrintString "Not Carrying Player :-(" false false)
                (Weapon|SetHolsterWeapon _playerpawn 0 false true _holsterid)))))
        (:Added
          (Development|PrintString "Carrying Player!" false false)
          (Weapon|SetHolsterWeapon _playerpawn 0 true true _holsterid))))
    (:then_2
      (CallFunction|StartHUDActionKeyManager))))

(event Custom|OnDied_Event_0
  (params (DamagedActor :type Actor%20Object%20Reference) (Damage :type Float%20%28single-precision%29) (InstigatedBy :type Controller%20Object%20Reference) (DamageCauser :type Actor%20Object%20Reference) (HitLocation :type Vector) (FHitComponent :type Primitive%20Component%20Object%20Reference) (BoneName :type Name) (Momentum :type Vector))
  ;[1 1 1] End the ability when the player dies
  (Ability|EndAbility))

(event Custom|Movement%20Mode%20Changed%20Delegate
  (params (Character :type Character%20Object%20Reference) (PrevMovementMode :type EMovementMode%20Enum :default "MOVE_None") (PreviousCustomMode :type Byte))
  (bind _playerpawn (Variables|Default|GetPlayerPawn))
  ;[1 1 1] In Water Behavior (keep the Lobster holstered, even while dolphin diving and water jumping)
  (if (Utilities|Enum|Equal(Enum) (Utilities|Enum|BytetoEnumEFortCustomMovement (Variables|/Script/Engine.CharacterMovementComponent|GetCustomMovementMode (Variables|/Script/Engine.Character|GetCharacterMovement Character))) "SurfaceSwimming")
    (Ability|Tasks|WaitNetSync "BothWait"
      (:OnSync
        (Development|PrintString "Lobster In Water!" false false)
        ;;[1 1 1 1 12] Lightsaber hidden by design in water until we decide to pursue proper lightsaber-water interactions
        (Weapon|SetHolsterWeapon _playerpawn 0 true true (Variables|Default|GetHolsterID))
        (bind _returnvalue (Ability|ApplyGameplayEffectToOwner (Variables|Default|GetGE_InWater)))
        (Variables|Default|SetInWaterHandle _returnvalue)
        (EventDispatchers|AssignOnPawnLanded _playerpawn (delegate Custom|OnPawnLanded_Event_0))
        (bind _asynctaskproxy (Ability|Tasks|WaitGameplayTagQuery (Variables|Default|GetUnholsterFromWaterQuery) 0 "WhenTrue" true)
          (:then
            (Variables|Default|SetUnholsterFromWaterTask _asynctaskproxy))
          (:Triggered
            _))))
    (elif (Utilities|Enum|Equal(Enum) (Utilities|Enum|BytetoEnumEFortCustomMovement PreviousCustomMode) "SurfaceSwimming")
      (Development|PrintString "Lobster Exit Water" false false)
      (if (Variables|/Script/FortniteGame.FortPlayerPawn|GetbIsWaterJump _playerpawn)
        _
        (else
          _))))
  (Weapon|SetHolsterWeapon _playerpawn 0 false true (Variables|Default|GetHolsterID))
  (Ability|RemoveGameplayEffectFromOwnerWithHandle (Variables|Default|GetInWaterHandle))
  (bind _unholsterfromwatertask (Variables|Default|GetUnholsterFromWaterTask)
    (:then
      (GameplayTasks|EndTask _unholsterfromwatertask))
    (:else)))

(event Custom|OnPawnLanded_Event_0
  (params (Hit :type Hit%20Result%20Structure%20%28by%20ref%29))
  ;[1 1 1] In Water Behavior (keep the Lobster holstered, even while dolphin diving and water jumping)
  (Weapon|SetHolsterWeapon (Variables|Default|GetPlayerPawn) 0 false true (Variables|Default|GetHolsterID))
  (Ability|RemoveGameplayEffectFromOwnerWithHandle (Variables|Default|GetInWaterHandle))
  (bind _unholsterfromwatertask (Variables|Default|GetUnholsterFromWaterTask)
    (:then
      (GameplayTasks|EndTask _unholsterfromwatertask))
    (:else)))

(event Ability|EventOnEndAbility
  (params (bWasCancelled :type Boolean))
  (bind _playerpawn (Variables|Default|GetPlayerPawn))
  ;[1 1 1] End Ability Cleanup
  (Ability|RemoveGameplayEffectFromOwnerWithHandle (Variables|Default|GetLoopingAudioHandle))
  (Ability|RemoveGameplayEffectFromOwnerWithHandle (Variables|Default|GetInWaterHandle))
  (Weapon|SetHolsterWeapon _playerpawn 0 false false (Variables|Default|GetHolsterID))
  (Character|UnbindEventfromMovementModeChangedDelegate _playerpawn (EventDispatchers|CreateEvent :!event Custom|Movement%20Mode%20Changed%20Delegate))
  (EventDispatchers|UnbindEventfromOnPawnLanded _playerpawn (EventDispatchers|CreateEvent :!event Custom|OnPawnLanded_Event_0)))

(event Ability|EventFailedToActivatePassiveAbility
  (params (ActorInfo :type Gameplay%20Ability%20Actor%20Info%20Structure))
  (|CallParentFunction))

(event Custom|StartHUDActionKeyManager
  (Ability|HasAuthority
    (:False
      (for _array_element (GameplayTags|BreakGameplayTagContainer (Variables|Default|GetTags_UpdateHUDActionKeys))
        (Ability|Tasks|WaitGameplayTagAdd _array_element
          (:then
            (Ability|Tasks|WaitGameplayTagRemove _array_element
              (:then
                (CallFunction|UpdateHUDActionKeys))
              (:Removed
                (CallFunction|UpdateHUDActionKeys))))
          (:Added
            (CallFunction|UpdateHUDActionKeys)))))))

(event Custom|UpdateHUDActionKeys
  (bind _asb_athena_galileo_lobster_parent (Utilities|Casting|CastToB_Athena_Galileo_Lobster_Parent (Weapon|GetCurrentWeapon (Variables|Default|GetPlayerPawn)))
    (:then
      (Class|BAthenaGalileoLobsterParent|ToggleAttackPrompt _asb_athena_galileo_lobster_parent (not (GameplayTags|HasAnyMatchingGameplayTags (Variables|Default|GetPlayerPawn) (Variables|Default|GetTags_HudActionKeys_Hide_Attack))))
      (Class|BAthenaGalileoLobsterParent|ToggleBlockPrompt _asb_athena_galileo_lobster_parent (not (GameplayTags|HasAnyMatchingGameplayTags (Variables|Default|GetPlayerPawn) (Variables|Default|GetTags_HudActionKeys_Hide_Block))))
      (Class|BAthenaGalileoLobsterParent|ToggleDodgePrompt _asb_athena_galileo_lobster_parent (and (not (GameplayTags|HasAnyMatchingGameplayTags (Variables|Default|GetPlayerPawn) (Variables|Default|GetTags_HudActionKeys_Hide_Dodge))) (GameplayTags|HasMatchingGameplayTag (Variables|Default|GetPlayerPawn) "(TagName=\"Gameplay.Action.Player.Melee.Block\")"))))
    (:CastFailed
      (Utilities|Time|SetTimerforNextTickbyEvent (delegate Custom|UpdateHUDActionKeys)))))
